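# Animation Blueprint for the AkimboEnvoy backpack, adding physics sway to its unicorn, two straps and zipper.
# Anim Dynamics simulates the zipper on the input pose and caches it as BasePose, feeding four differently tuned unicorn-and-strap chains.
# Nested bool blends select the downed (DBNO), front-end, skydiving or default chain.
# The downed chain first additively offsets the unicorn bone, while the front-end chain simulates the zipper a second time.
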AnimBlueprint /BRCosmetics/Accessories/FORT_Backpacks/Backpack_F_MED_AkimboEnvoy/Meshes/F_MED_AkimboEnvoy_Pack_AnimBP
target_skeleton=&/F_MED_AkimboEnvoy_Pack_Skeleton
graphs [
 AnimGraph {nodes [
   n1 = "Transform (Modify) Bone - Bone: dyn_unicorn"(
    ComponentPose=Local To Component(LocalPose=Use cached pose 'BasePose'())
    Alpha=1
    Translation="-0.2,0.2,-0.4"
    Rotation="0,0,5"
    Scale="1,1,1"
   ) {
    rotationMode=BMM_Additive
    rotationSpace=BCS_BoneSpace
    scaleMode=BMM_Additive
    scaleSpace=BCS_BoneSpace
    translationMode=BMM_Additive
    translationSpace=BCS_BoneSpace
    alphaBoolBlend {blendOption=Linear}
    boneToModify {boneName=dyn_unicorn}
    rotation {roll=5}
    translation=<-0.2 0.2 -0.4>
   }
   n2 = Anim Dynamics - Bone: dyn_unicorn!(ComponentPose=@n1, Alpha=0.7, GravityOverride:Vector="0,0,-980", ExternalForce="0,0,0") -> (Pose) {
    angularBiasOverride=1
    angularDampingOverride=1
    angularSpringConstant=3000
    bAngularSpring=true
    bOverrideAngularBias=true
    bOverrideAngularDamping=true
    bUseGravityOverride=true
    bUsePlanarLimit=false
    simulationSpace=RootRelative
    alphaBoolBlend {blendOption=Linear}
    boundBone {boneName=dyn_unicorn}
    componentAppliedLinearAccClamp=<3000 3000 3000>
    componentLinearAccScale=<0.4 0.5 0.5>
    componentLinearVelScale=<0.5 0.5 0.5>
    physicsBodyDefinitions [{
      collisionType=CustomSphere
      sphereCollisionRadius=2.2
      boundBone {boneName=dyn_unicorn}
      boxExtents=<10 8 8>
      constraintSetup {angularLimitsMax=<30 130 0>, angularLimitsMin=<-30 -130 -130>, angularTarget=<0 0 0>}
      localJointOffset=<2 0 0>
     }]
   }
   n3 = Anim Dynamics - Bone: dyn_strap_01!(ComponentPose=@n2, Alpha=0.7, GravityOverride:Vector="0,0,-980", ExternalForce="0,0,0") -> (Pose) {
    angularBiasOverride=1
    angularDampingOverride=1
    angularSpringConstant=3000
    bAngularSpring=true
    bOverrideAngularBias=true
    bOverrideAngularDamping=true
    bUseGravityOverride=true
    bUsePlanarLimit=false
    simulationSpace=RootRelative
    alphaBoolBlend {blendOption=Linear}
    boundBone {boneName=dyn_strap_01}
    componentAppliedLinearAccClamp=<3000 3000 3000>
    componentLinearAccScale=<0.4 0.5 0.5>
    componentLinearVelScale=<0.5 0.5 0.5>
    physicsBodyDefinitions [{
      collisionType=CustomSphere
      sphereCollisionRadius=2.2
      boundBone {boneName=dyn_strap_01}
      boxExtents=<10 8 8>
      constraintSetup {angularLimitsMax=<5 35 0>, angularLimitsMin=<-3 -30 -130>, angularTarget=<0 0 0>}
      localJointOffset=<2 0 0>
     }]
    planarLimits [{drivingBone {boneName=pack}, planeTransform {location=<-0.4138 4.056 0>, rotation=<rot 0.0001 10 -90.0003>}}]
   }
   n4 = Anim Dynamics - Bone: dyn_strap_02!(ComponentPose=@n3, Alpha=0.7, GravityOverride:Vector="0,0,-980", ExternalForce="0,0,0") -> (Pose) {
    angularBiasOverride=1
    angularDampingOverride=1
    angularSpringConstant=3000
    bAngularSpring=true
    bOverrideAngularBias=true
    bOverrideAngularDamping=true
    bUseGravityOverride=true
    bUsePlanarLimit=false
    simulationSpace=RootRelative
    alphaBoolBlend {blendOption=Linear}
    boundBone {boneName=dyn_strap_02}
    componentAppliedLinearAccClamp=<3000 3000 3000>
    componentLinearAccScale=<0.4 0.5 0.5>
    componentLinearVelScale=<0.5 0.5 0.5>
    physicsBodyDefinitions [{
      collisionType=CustomSphere
      sphereCollisionRadius=3
      boundBone {boneName=dyn_strap_02}
      boxExtents=<10 7 7>
      constraintSetup {angularLimitsMax=<5 45 45>, angularLimitsMin=<-5 -45 -80>, angularTarget=<0 0 0>}
      localJointOffset=<2 0 0>
     }]
    planarLimits [{drivingBone {boneName=pack}, planeTransform {location=<-0.5576 4.8711 0>, rotation=<rot 0.0001 10 -90.0003>}}]
   }
   Output Pose(Result=Blend Poses by bool(BlendPose_0=Component To Local(ComponentPose=@n4), BlendPose_1=Use cached pose 'OutputPose'(), bActiveValue=Get bIsDBNO()) {blendType=Linear}) {layerGroup=None}
   n5 = Anim Dynamics - Bone: dyn_zipper(ComponentPose=Local To Component(LocalPose=Input Pose()), Alpha=1, ExternalForce="0,0,-980") {
    angularBiasOverride=0.99
    angularDampingOverride=1
    angularSpringConstant=500
    bAngularSpring=true
    bOverrideAngularBias=true
    bOverrideAngularDamping=true
    bUseGravityOverride=true
    gravityScale=0
    simulationSpace=RootRelative
    alphaBoolBlend {blendOption=Linear}
    boundBone {boneName=dyn_zipper}
    componentAppliedLinearAccClamp=<3000 3000 3000>
    componentLinearAccScale=<0.45 0.4 0.45>
    componentLinearVelScale=<0.7 0.6 0.7>
    gravityOverride {z=-980}
    physicsBodyDefinitions [{
      boundBone {boneName=dyn_zipper}
      boxExtents=<5 3 5.5>
      constraintSetup {angularLimitsMax=<5 35 0>, angularLimitsMin=<-3 -30 -30>, angularTarget=<1 0 0>}
      localJointOffset=<1 0 0>
     }]
    relativeSpaceBone {boneName=root}
   }
   Save cached pose 'BasePose'(Pose=Component To Local(ComponentPose=@n5)) {cacheName=BasePose}
   n6 = Anim Dynamics - Bone: dyn_unicorn!(ComponentPose=Local To Component(LocalPose=Use cached pose 'BasePose'()), Alpha=0.7, ExternalForce="0,0,-980") -> (Pose) {
    angularBiasOverride=1
    angularDampingOverride=1
    angularSpringConstant=500
    bAngularSpring=true
    bOverrideAngularBias=true
    bOverrideAngularDamping=true
    bUseGravityOverride=true
    bUsePlanarLimit=false
    gravityScale=0
    simulationSpace=RootRelative
    alphaBoolBlend {blendOption=Linear}
    boundBone {boneName=dyn_unicorn}
    componentAppliedLinearAccClamp=<3000 3000 3000>
    componentLinearAccScale=<0.45 0.4 0.45>
    componentLinearVelScale=<0.7 0.6 0.7>
    gravityOverride {z=-980}
    physicsBodyDefinitions [{
      boundBone {boneName=dyn_unicorn}
      boxExtents=<5 3 5>
      constraintSetup {angularLimitsMax=<0 30 3>, angularLimitsMin=<0 0 -60>, angularTarget=<1 0 0>}
      localJointOffset=<1 0 0>
     }]
    relativeSpaceBone {boneName=root}
   }
   n7 = Anim Dynamics - Bone: dyn_strap_01(ComponentPose=@n6, Alpha=0.7, GravityOverride="0,0,-980", ExternalForce="0,0,0") {
    angularBiasOverride=1
    angularDampingOverride=1
    angularSpringConstant=3000
    bAngularSpring=true
    bOverrideAngularBias=true
    bOverrideAngularDamping=true
    bUseGravityOverride=true
    bUsePlanarLimit=false
    simulationSpace=RootRelative
    alphaBoolBlend {blendOption=Linear}
    boundBone {boneName=dyn_strap_01}
    componentAppliedLinearAccClamp=<3000 3000 3000>
    componentLinearAccScale=<0.4 0.5 0.5>
    componentLinearVelScale=<0.5 0.5 0.5>
    physicsBodyDefinitions [{
      collisionType=CustomSphere
      sphereCollisionRadius=2.2
      boundBone {boneName=dyn_strap_01}
      boxExtents=<10 8 8>
      constraintSetup {angularLimitsMax=<5 35 0>, angularLimitsMin=<-3 -30 -130>, angularTarget=<0 0 0>}
      localJointOffset=<2 0 0>
     }]
   }
   n8 = Anim Dynamics - Bone: dyn_strap_02(ComponentPose=@n7, Alpha=0.7, GravityOverride="0,0,-980", ExternalForce="0,0,0") {
    angularBiasOverride=1
    angularDampingOverride=1
    angularSpringConstant=3000
    bAngularSpring=true
    bOverrideAngularBias=true
    bOverrideAngularDamping=true
    bUseGravityOverride=true
    bUsePlanarLimit=false
    simulationSpace=RootRelative
    alphaBoolBlend {blendOption=Linear}
    boundBone {boneName=dyn_strap_02}
    componentAppliedLinearAccClamp=<3000 3000 3000>
    componentLinearAccScale=<0.4 0.5 0.5>
    componentLinearVelScale=<0.5 0.5 0.5>
    physicsBodyDefinitions [{
      collisionType=CustomSphere
      sphereCollisionRadius=3
      boundBone {boneName=dyn_strap_02}
      boxExtents=<10 7 7>
      constraintSetup {angularLimitsMax=<5 45 45>, angularLimitsMin=<-5 -45 -80>, angularTarget=<0 0 0>}
      localJointOffset=<2 0 0>
     }]
   }
   n9 = Component To Local(ComponentPose=@n8)
   n10 = Anim Dynamics - Bone: dyn_unicorn(ComponentPose=Local To Component(LocalPose=Use cached pose 'BasePose'()), Alpha=0.7, ExternalForce="0,0,-980") {
    angularBiasOverride=0.99
    angularDampingOverride=1
    angularSpringConstant=500
    bAngularSpring=true
    bOverrideAngularBias=true
    bOverrideAngularDamping=true
    bUseGravityOverride=true
    bUsePlanarLimit=false
    gravityScale=0
    simulationSpace=RootRelative
    alphaBoolBlend {blendOption=Linear}
    boundBone {boneName=dyn_unicorn}
    componentAppliedLinearAccClamp=<3000 3000 3000>
    componentLinearAccScale=<0.45 0.4 0.45>
    componentLinearVelScale=<0.7 0.6 0.7>
    gravityOverride {z=-980}
    physicsBodyDefinitions [{
      boundBone {boneName=dyn_unicorn}
      boxExtents=<5 3 5>
      constraintSetup {angularLimitsMax=<0 0 3>, angularLimitsMin=<0 0 -30>, angularTarget=<1 0 0>}
      localJointOffset=<1 0 0>
     }]
    relativeSpaceBone {boneName=root}
   }
   n11 = Anim Dynamics - Bone: dyn_strap_01(ComponentPose=@n10, Alpha=0.7, GravityOverride="0,0,-980", ExternalForce="0,0,0") {
    angularBiasOverride=1
    angularDampingOverride=1
    angularSpringConstant=3000
    bAngularSpring=true
    bOverrideAngularBias=true
    bOverrideAngularDamping=true
    bUseGravityOverride=true
    simulationSpace=RootRelative
    alphaBoolBlend {blendOption=Linear}
    boundBone {boneName=dyn_strap_01}
    componentAppliedLinearAccClamp=<3000 3000 3000>
    componentLinearAccScale=<0.4 0.5 0.5>
    componentLinearVelScale=<0.5 0.5 0.5>
    physicsBodyDefinitions [{
      collisionType=CustomSphere
      sphereCollisionRadius=2.2
      boundBone {boneName=dyn_strap_01}
      boxExtents=<10 14 14>
      constraintSetup {angularLimitsMax=<5 35 0>, angularLimitsMin=<-3 -30 -130>, angularTarget=<0 0 0>}
      localJointOffset=<5 0 0>
     }]
    planarLimits [{drivingBone {boneName=pack}, planeTransform {location=<0.354 -0.2986 0>, rotation=<rot 0.0001 10 -90.0003>}}]
   }
   n12 = Anim Dynamics - Bone: dyn_strap_02(ComponentPose=@n11, Alpha=0.7, GravityOverride="0,0,-980", ExternalForce="0,0,0") {
    angularBiasOverride=1
    angularDampingOverride=1
    angularSpringConstant=3000
    bAngularSpring=true
    bOverrideAngularBias=true
    bOverrideAngularDamping=true
    bUseGravityOverride=true
    simulationSpace=RootRelative
    alphaBoolBlend {blendOption=Linear}
    boundBone {boneName=dyn_strap_02}
    componentAppliedLinearAccClamp=<3000 3000 3000>
    componentLinearAccScale=<0.4 0.5 0.5>
    componentLinearVelScale=<0.5 0.5 0.5>
    physicsBodyDefinitions [{
      collisionType=CustomSphere
      sphereCollisionRadius=3
      boundBone {boneName=dyn_strap_02}
      boxExtents=<18 12 12>
      constraintSetup {angularLimitsMax=<5 45 45>, angularLimitsMin=<-5 -45 -80>, angularTarget=<0 0 0>}
      localJointOffset=<6 0 0>
     }]
    planarLimits [{drivingBone {boneName=pack}, planeTransform {location=<0.152 0.847 0>, rotation=<rot 0.0001 10 -90.0003>}}]
   }
   n13 = Component To Local(ComponentPose=@n12)
   n14 = Get bIsSkyDiving()
   n15 = Anim Dynamics - Bone: dyn_unicorn(ComponentPose=Local To Component(LocalPose=Use cached pose 'BasePose'()), Alpha=0.7, ExternalForce="0,0,-980") {
    angularBiasOverride=0.99
    angularDampingOverride=1
    angularSpringConstant=500
    bAngularSpring=true
    bOverrideAngularBias=true
    bOverrideAngularDamping=true
    bUseGravityOverride=true
    bUsePlanarLimit=false
    gravityScale=0
    simulationSpace=RootRelative
    alphaBoolBlend {blendOption=Linear}
    boundBone {boneName=dyn_unicorn}
    componentAppliedLinearAccClamp=<3000 3000 3000>
    componentLinearAccScale=<0.45 0.4 0.45>
    componentLinearVelScale=<0.7 0.6 0.7>
    gravityOverride {z=-980}
    physicsBodyDefinitions [{
      boundBone {boneName=dyn_unicorn}
      boxExtents=<8 6 8>
      constraintSetup {angularLimitsMax=<0 0 3>, angularLimitsMin=<0 0 -30>, angularTarget=<1 0 0>}
      localJointOffset=<1 0 0>
     }]
    relativeSpaceBone {boneName=root}
   }
   n16 = Anim Dynamics - Bone: dyn_strap_01(ComponentPose=@n15, Alpha=0.7, GravityOverride="0,0,-980", ExternalForce="0,0,0") {
    angularBiasOverride=1
    angularDampingOverride=1
    angularSpringConstant=3000
    bAngularSpring=true
    bOverrideAngularBias=true
    bOverrideAngularDamping=true
    bUseGravityOverride=true
    simulationSpace=RootRelative
    alphaBoolBlend {blendOption=Linear}
    boundBone {boneName=dyn_strap_01}
    componentAppliedLinearAccClamp=<3000 3000 3000>
    componentLinearAccScale=<0.4 0.5 0.5>
    componentLinearVelScale=<0.5 0.5 0.5>
    physicsBodyDefinitions [{
      collisionType=CustomSphere
      sphereCollisionRadius=2.2
      boundBone {boneName=dyn_strap_01}
      boxExtents=<10 14 14>
      constraintSetup {angularLimitsMax=<5 35 0>, angularLimitsMin=<-3 -30 -130>, angularTarget=<0 0 0>}
      localJointOffset=<5 0 0>
     }]
    planarLimits [{drivingBone {boneName=pack}, planeTransform {location=<0.354 -0.2986 0>, rotation=<rot 0.0001 10 -90.0003>}}]
   }
   n17 = Anim Dynamics - Bone: dyn_strap_02(ComponentPose=@n16, Alpha=0.7, GravityOverride="0,0,-980", ExternalForce="0,0,0") {
    angularBiasOverride=1
    angularDampingOverride=1
    angularSpringConstant=500
    bAngularSpring=true
    bOverrideAngularBias=true
    bOverrideAngularDamping=true
    bUseGravityOverride=true
    simulationSpace=RootRelative
    alphaBoolBlend {blendOption=Linear}
    boundBone {boneName=dyn_strap_02}
    componentAppliedLinearAccClamp=<3000 3000 3000>
    componentLinearAccScale=<0.4 0.5 0.5>
    componentLinearVelScale=<0.5 0.5 0.5>
    physicsBodyDefinitions [{
      collisionType=CustomSphere
      sphereCollisionRadius=3
      boundBone {boneName=dyn_strap_02}
      boxExtents=<18 13 13>
      constraintSetup {angularLimitsMax=<5 45 45>, angularLimitsMin=<-5 -45 -80>, angularTarget=<0 0 0>}
      localJointOffset=<6 0 0>
     }]
    planarLimits [{drivingBone {boneName=pack}, planeTransform {location=<0.152 0.847 0>, rotation=<rot 0.0001 10 -90.0003>}}]
   }
   n18 = Anim Dynamics - Bone: dyn_zipper(ComponentPose=@n17, Alpha=1, ExternalForce="0,0,-980") {
    angularBiasOverride=0.99
    angularDampingOverride=1
    angularSpringConstant=300
    bAngularSpring=true
    bOverrideAngularBias=true
    bOverrideAngularDamping=true
    bUseGravityOverride=true
    gravityScale=0
    simulationSpace=RootRelative
    alphaBoolBlend {blendOption=Linear}
    boundBone {boneName=dyn_zipper}
    componentAppliedLinearAccClamp=<3000 3000 3000>
    componentLinearAccScale=<0.45 0.4 0.45>
    componentLinearVelScale=<0.7 0.6 0.7>
    gravityOverride {z=-980}
    physicsBodyDefinitions [{
      boundBone {boneName=dyn_zipper}
      boxExtents=<7 7 7>
      constraintSetup {angularLimitsMax=<5 35 0>, angularLimitsMin=<-3 -30 -30>, angularTarget=<1 0 0>}
      localJointOffset=<1 0 0>
     }]
    relativeSpaceBone {boneName=root}
   }
   Save cached pose 'OutputPose'(Pose=Blend Poses by bool(
     BlendPose_0=Component To Local(ComponentPose=@n18)
     BlendPose_1=Blend Poses by bool(BlendPose_0=@n13, BlendPose_1=@n9, bActiveValue=@n14) {blendType=Linear}
     bActiveValue=Get bIsInFrontEnd()
    ) {blendType=Linear}) {cacheName=OutputPose}
  ]}
 EventGraph {nodes [Event BlueprintUpdateAnimation()]}
]
variables []
